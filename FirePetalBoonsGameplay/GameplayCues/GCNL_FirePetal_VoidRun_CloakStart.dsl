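; Looping gameplay cue for the FirePetal Void Run cloak that pulses the target player pawn's visibility between minimum and maximum.
; On loop start it begins decreasing toward minimum and starts a fast looping timer that steps a Decreasing, Retaining, Increasing boolean state machine.
; Transitions lerp OverallOpacity on the pawn's overridden materials by elapsed time, with separate durations per direction and separate hold times at each extreme.
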(event EventOnLoopingStart(EitherVFX-preferthisversion)
  (params (MyTarget :type Actor%20Object%20Reference) (Parameters :type Gameplay%20Cue%20Parameters%20Structure%20%28by%20ref%29) (ParticleComponents :type Array%20of%20FXSystem%20Component%20Object%20References) (AudioComponents :type Array%20of%20Audio%20Component%20Object%20References))
  (bind _my_target (Variables|Default|GetMy%20Target))
  (bind _returnvalue (Utilities|Time|GetTimeSeconds _my_target))
  (bind _output_get (Variables|Default|SetMy%20Target MyTarget))
  (|CallParentFunction _output_get Parameters ParticleComponents AudioComponents)
  (Variables|Default|SetTransitionGoalTimeStart _returnvalue)
  (Variables|Default|SetTransitionGoalTimeEnd (+ _returnvalue (Variables|Default|GetTransitionTimeMaxToMin)))
  (Variables|Default|SetTransitionGoalVisibility (ScalableFloat|GetValueasDouble (Variables|Default|GetMinVisibility)))
  (Variables|Default|SetDecreasing true)
  (Variables|Default|SetIncreasing false)
  (Variables|Default|SetRetaining false)
  (Utilities|Time|SetTimerbyEvent (delegate Custom|StepInvis) 0.01 true true))

(event Custom|StepInvis
  (if (Variables|Default|GetRetaining)
    (CallFunction|RetainVisibility)
    (elif (Variables|Default|GetIncreasing)
      (CallFunction|IncreaseVisibility)
      (elif (Variables|Default|GetDecreasing)
        (CallFunction|DecreaseVisibility)))))

(fn DecreaseVisibility
  (bind _my_target (Variables|Default|GetMy%20Target))
  (bind _returnvalue (Utilities|Time|GetTimeSeconds _my_target))
  ; Are we done 'Decreasing'?
  (if (>= (Utilities|Time|GetTimeSeconds (Variables|Default|GetMy%20Target)) (Variables|Default|GetTransitionGoalTimeEnd))
    ; This transition of "Decreasing" is over. Enter a "Retain" state to hold this visibility for some time.
    (bind _output_get (Variables|Default|SetScalarParamOpacity (ScalableFloat|GetValueasDouble (Variables|Default|GetMinVisibility))))
    (bind _asfort_player_pawn (Utilities|Casting|CastToFortPlayerPawn (Variables|Default|GetMy%20Target)))
    (Class|FortPlayerPawn|SetScalarParamonOverriddenMaterials _asfort_player_pawn "OverallOpacity" _output_get)
    (Variables|Default|SetRetaining true)
    (Variables|Default|SetDecreasing false)
    (Variables|Default|SetTransitionGoalTimeStart _returnvalue)
    (Variables|Default|SetTransitionGoalTimeEnd (+ _returnvalue (Variables|Default|GetRetainTime)))
    (else
      ; Keep "Decreasing". Get a lerped value of visibility based on how long this transition has been occurring.
      (bind _output_get_2 (Variables|Default|SetScalarParamOpacity (Math|Float|Lerp (ScalableFloat|GetValueasDouble (Variables|Default|GetMaxVisibility)) (ScalableFloat|GetValueasDouble (Variables|Default|GetMinVisibility)) (CallFunction|GetVisibilityCurrentLerpValue))))
      (bind _asfort_player_pawn_2 (Utilities|Casting|CastToFortPlayerPawn (Variables|Default|GetMy%20Target)))
      (Class|FortPlayerPawn|SetScalarParamonOverriddenMaterials _asfort_player_pawn_2 "OverallOpacity" _output_get_2))))

(fn GetVisibilityCurrentLerpValue
  (return (/ (- (Utilities|Time|GetTimeSeconds (Variables|Default|GetMy%20Target)) (Variables|Default|GetTransitionGoalTimeStart)) (- (Variables|Default|GetTransitionGoalTimeEnd) (Variables|Default|GetTransitionGoalTimeStart)))))

(fn IncreaseVisibility
  (bind _my_target (Variables|Default|GetMy%20Target))
  (bind _returnvalue (Utilities|Time|GetTimeSeconds _my_target))
  ; Are we done 'Increasing'?
  (if (>= (Utilities|Time|GetTimeSeconds (Variables|Default|GetMy%20Target)) (Variables|Default|GetTransitionGoalTimeEnd))
    ; This transition of "Increasing" is over. Enter a "Retain" state to hold this visibility for some time.
    (bind _output_get (Variables|Default|SetScalarParamOpacity (ScalableFloat|GetValueasDouble (Variables|Default|GetMaxVisibility))))
    (bind _asfort_player_pawn (Utilities|Casting|CastToFortPlayerPawn (Variables|Default|GetMy%20Target)))
    (Class|FortPlayerPawn|SetScalarParamonOverriddenMaterials _asfort_player_pawn "OverallOpacity" _output_get)
    (Variables|Default|SetRetaining true)
    (Variables|Default|SetIncreasing false)
    (Variables|Default|SetTransitionGoalTimeStart _returnvalue)
    (Variables|Default|SetTransitionGoalTimeEnd (+ _returnvalue (Variables|Default|GetRetainTimeMax)))
    (else
      ; Keep "Increasing". Get a lerped value of visibility based on how long this transition has been occurring.
      (bind _output_get_2 (Variables|Default|SetScalarParamOpacity (Math|Float|Lerp (ScalableFloat|GetValueasDouble (Variables|Default|GetMinVisibility)) (ScalableFloat|GetValueasDouble (Variables|Default|GetMaxVisibility)) (CallFunction|GetVisibilityCurrentLerpValue))))
      (bind _asfort_player_pawn_2 (Utilities|Casting|CastToFortPlayerPawn (Variables|Default|GetMy%20Target)))
      (Class|FortPlayerPawn|SetScalarParamonOverriddenMaterials _asfort_player_pawn_2 "OverallOpacity" _output_get_2))))

(fn RetainVisibility
  (bind _my_target (Variables|Default|GetMy%20Target))
  (bind _returnvalue (Utilities|Time|GetTimeSeconds _my_target))
  (bind _my_target_2 (Variables|Default|GetMy%20Target))
  (bind _returnvalue_2 (Utilities|Time|GetTimeSeconds _my_target_2))
  ; Are we done retaining yet?
  (if (>= (Utilities|Time|GetTimeSeconds (Variables|Default|GetMy%20Target)) (Variables|Default|GetTransitionGoalTimeEnd))
    (Variables|Default|SetRetaining false)
    (if (== (Variables|Default|GetScalarParamOpacity) (Variables|Default|GetMaxVisibility))
      ; Transition to and begin a "Decreasing" state of visibility
      (Variables|Default|SetDecreasing true)
      (Variables|Default|SetTransitionGoalTimeStart _returnvalue_2)
      (Variables|Default|SetTransitionGoalTimeEnd (+ _returnvalue_2 (Variables|Default|GetTransitionTimeMaxToMin)))
      (Variables|Default|SetTransitionGoalVisibility (ScalableFloat|GetValueasDouble (Variables|Default|GetMinVisibility)))
      (else
        ; Transition to and begin an "Increasing" state of visibility
        (Variables|Default|SetIncreasing true)
        (Variables|Default|SetTransitionGoalTimeStart _returnvalue)
        (Variables|Default|SetTransitionGoalTimeEnd (+ _returnvalue (Variables|Default|GetTransitionTimeMinToMax)))
        (Variables|Default|SetTransitionGoalVisibility (ScalableFloat|GetValueasDouble (Variables|Default|GetMaxVisibility)))))))

(fn ConstructionScript)
