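; Base looping gameplay cue for the active Hot Hands boon, adding speed-driven Niagara effects and an expiry warning sound on the target.
; Loop start attaches the speed effect and caches max speed from a scalable float; removal deactivates it.
; Speed updates map clamped speed over that maximum through a curve into an effect parameter.
; Remaining-duration updates play the warning once below a threshold, rearming when duration rises above it.
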(event EventOnApplication(EitherVFX-preferthisone)
  (params (MyTarget :type Actor%20Object%20Reference) (Parameters :type Gameplay%20Cue%20Parameters%20Structure%20%28by%20ref%29) (ParticleComponents :type Array%20of%20FXSystem%20Component%20Object%20References) (AudioComponents :type Array%20of%20Audio%20Component%20Object%20References) (BurstCameraShakeInstance :type Camera%20Shake%20Base%20Object%20Reference) (BurstDecalInstance :type Decal%20Actor%20Object%20Reference))
  (|CallParentFunction MyTarget Parameters ParticleComponents AudioComponents BurstCameraShakeInstance BurstDecalInstance))

(event EventOnLoopingStart(EitherVFX-preferthisversion)
  (params (MyTarget :type Actor%20Object%20Reference) (Parameters :type Gameplay%20Cue%20Parameters%20Structure%20%28by%20ref%29) (ParticleComponents :type Array%20of%20FXSystem%20Component%20Object%20References) (AudioComponents :type Array%20of%20Audio%20Component%20Object%20References))
  (|CallParentFunction MyTarget Parameters ParticleComponents AudioComponents)
  ;[1 1 1] Spawn speed FX and initialize with 0 speed. This will be set by UpdateSpeed
  (bind _returnvalue (Niagara|SpawnSystemAttached (Variables|Speed|GetSpeedFXSystem) (Variables|/Script/Engine.Actor|GetRootComponent MyTarget) "None" "0, 0, 0" "0, 0, 0" "SnapToTarget" true))
  (bind _output_get (Variables|Speed|SetSpawnedSpeedFX _returnvalue))
  (Effects|Components|ParticleSystem|SetFloatParameter _output_get "NormalizedSpeed" 0.0)
  (Variables|Speed|SetMaxSpeed (ScalableFloat|GetValueatLevel (Variables|Speed|GetRow_MinExplosiveSpeed)))
  (Effects|Components|ParticleSystem|SetFloatParameter (Variables|Speed|GetSpawnedSpeedFX) "RemainingDuration" 2e+01) ; Initializing the RemainingDuration parameter
)

(event EventOnRemoval(EitherVFX-preferthisversion)
  (params (MyTarget :type Actor%20Object%20Reference) (Parameters :type Gameplay%20Cue%20Parameters%20Structure%20%28by%20ref%29) (ParticleComponents :type Array%20of%20FXSystem%20Component%20Object%20References) (AudioComponents :type Array%20of%20Audio%20Component%20Object%20References) (BurstCameraShakeInstance :type Camera%20Shake%20Base%20Object%20Reference) (BurstDecalInstance :type Decal%20Actor%20Object%20Reference))
  (|CallParentFunction MyTarget Parameters ParticleComponents AudioComponents BurstCameraShakeInstance BurstDecalInstance)
  (bind _spawnedspeedfx (Variables|Speed|GetSpawnedSpeedFX)
    (:then
      (Components|Activation|Deactivate _spawnedspeedfx))
    (:else)))

(event EventUpdateRemainingDuration
  (params (NewRemainingDuration :type Float%20%28single-precision%29))
  (|CallParentFunction NewRemainingDuration)
  (if (< NewRemainingDuration (Variables|Default|GetTimeRemainingToWarn))
    (if (not (Variables|Default|GetbHasPlayedWarning))
      (Variables|Default|SetbHasPlayedWarning true)
      ;[1 1 1] Change  Node
      (bind _targetactor (Variables|FortGCNL_Nitro_Active_Base|GetTargetActor)
        (:then
          (Audio|SpawnSoundAttached (Variables|Default|GetWarningSound) (Variables|/Script/Engine.Actor|GetRootComponent _targetactor)))
        (:else)))
    (elif (Variables|Default|GetbHasPlayedWarning)
      ;[1 1 1] If we are above the threshold but have already played the warning cue (probably means we got refreshed), clear the flag so it can play again when we get low the next time
      (Variables|Default|SetbHasPlayedWarning false))))

(event EventUpdateCurrentSpeed
  (params (NewCurrentSpeed :type Float%20%28single-precision%29))
  (|CallParentFunction NewCurrentSpeed)
  (bind _output_get (Variables|Speed|SetCurrentSpeed NewCurrentSpeed))
  (Utilities|IsValid (Variables|Speed|GetSpawnedSpeedFX)
    (:Is%20Valid
      (Effects|Components|ParticleSystem|SetFloatParameter (Variables|Speed|GetSpawnedSpeedFX) (Variables|Speed|GetSpeedParamName) (Math|Curves|GetFloatValue (Variables|Speed|GetSpeedFXCurve) (Math|Float|Clamp(Float) (/ _output_get (Variables|Speed|GetMaxSpeed))))))))

(fn ConstructionScript
  (|CallParentFunction))
