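; Gameplay cue for the Hot Hands boon that shows an effect and plays a sound on the player pawn during reloads.
; On application, two async tag queries watch Gameplay.Action.Player.Reload: one spawns mesh-attached Niagara and audio, the other deactivates the effect and fades the audio.
; On removal it ends the reload-start wait and deactivates the effect, but neither fades the audio nor directly ends the reload-finished wait.
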(event EventOnApplication(EitherVFX-preferthisone)
  (params (MyTarget :type Actor%20Object%20Reference) (Parameters :type Gameplay%20Cue%20Parameters%20Structure%20%28by%20ref%29) (ParticleComponents :type Array%20of%20FXSystem%20Component%20Object%20References) (AudioComponents :type Array%20of%20Audio%20Component%20Object%20References) (BurstCameraShakeInstance :type Camera%20Shake%20Base%20Object%20Reference) (BurstDecalInstance :type Decal%20Actor%20Object%20Reference))
  (|CallParentFunction MyTarget Parameters ParticleComponents AudioComponents BurstCameraShakeInstance BurstDecalInstance)
  (bind _output_get (Variables|Default|SetFortPlayerPawn (Utilities|Casting|CastToFortPlayerPawn MyTarget)))
  (bind _asynctaskproxy (Ability|Tasks|WaitGameplayTagQueryOnActor _output_get "(TokenStreamVersion=0,TagDictionary=((TagName=\"Gameplay.Action.Player.Reload\")),QueryTokenStream=(0,1,1,1,0),UserDescription=\"\",AutoDescription=\" ANY( Gameplay.Action.Player.Reload )\")")
    (:then
      (Variables|Default|SetWaitReloadAsync _asynctaskproxy)
      (bind _asynctaskproxy_2 (Ability|Tasks|WaitGameplayTagQueryOnActor (Variables|Default|GetFortPlayerPawn) "(TokenStreamVersion=0,TagDictionary=((TagName=\"Gameplay.Action.Player.Reload\")),QueryTokenStream=(0,1,1,1,0),UserDescription=\"\",AutoDescription=\" ANY( Gameplay.Action.Player.Reload )\")" "WhenFalse")
        (:then
          (Variables|Default|SetReloadFinishedAsync _asynctaskproxy_2))
        (:Triggered
          (CallFunction|PlayHotHandsAudio false)
          (bind _reloadingfx (Variables|Default|GetReloadingFX)
            (:then
              (Components|Activation|Deactivate _reloadingfx))
            (:else)))))
    (:Triggered
      (bind _returnvalue (Niagara|SpawnSystemAttached "/FirePetalBoonsGameplay/Assets/FX/NS_HotHandsState_Player_Reloading.NS_HotHandsState_Player_Reloading" (Variables|/Script/Engine.Character|GetMesh _output_get)))
      (Variables|Default|SetReloadingFX _returnvalue)
      (CallFunction|PlayHotHandsAudio true))))

(event EventOnLoopingStart(EitherVFX-preferthisversion)
  (params (MyTarget :type Actor%20Object%20Reference) (Parameters :type Gameplay%20Cue%20Parameters%20Structure%20%28by%20ref%29) (ParticleComponents :type Array%20of%20FXSystem%20Component%20Object%20References) (AudioComponents :type Array%20of%20Audio%20Component%20Object%20References))
  (|CallParentFunction MyTarget Parameters ParticleComponents AudioComponents))

(event EventOnRemoval(EitherVFX-preferthisversion)
  (params (MyTarget :type Actor%20Object%20Reference) (Parameters :type Gameplay%20Cue%20Parameters%20Structure%20%28by%20ref%29) (ParticleComponents :type Array%20of%20FXSystem%20Component%20Object%20References) (AudioComponents :type Array%20of%20Audio%20Component%20Object%20References) (BurstCameraShakeInstance :type Camera%20Shake%20Base%20Object%20Reference) (BurstDecalInstance :type Decal%20Actor%20Object%20Reference))
  (bind _waitreloadasync (Variables|Default|GetWaitReloadAsync))
  (|CallParentFunction MyTarget Parameters ParticleComponents AudioComponents BurstCameraShakeInstance BurstDecalInstance)
  (if (AsyncAction|IsActive (Variables|Default|GetReloadFinishedAsync))
    (AsyncAction|Cancel _waitreloadasync))
  (Ability|Async|EndAction _waitreloadasync)
  (Components|Activation|Deactivate (Variables|Default|GetReloadingFX)))

(event Custom|PlayHotHandsAudio
  (params (bPlay :type Boolean))
  ; Audio setup
  (if bPlay
    (bind _fortplayerpawn (Variables|Default|GetFortPlayerPawn)
      (:then
        (Variables|Default|GetAppliedAudioComponent
          (:then)
          (:else
            (bind _returnvalue (Audio|SpawnSoundAttached (Variables|Default|GetAppliedSound) (Variables|/Script/Engine.Character|GetMesh _fortplayerpawn)))
            (Variables|Default|SetAppliedAudioComponent _returnvalue))))
      (:else))
    (else
      (bind _appliedaudiocomponent (Variables|Default|GetAppliedAudioComponent)
        (:then
          (Audio|Components|Audio|FadeOut _appliedaudiocomponent 0.3 0.0 "Linear" :!class /Script/Engine.AudioComponent))
        (:else)))))

(fn ConstructionScript
  (|CallParentFunction))
